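# Physics asset for the Cuddle Team Dark pickaxe cosmetic, defining the bodies and joints for physical motion of its ribbon.
# Seven small zero-length capsule bodies, effectively spheres, form a single chain from dyn_ribbon_01 to dyn_ribbon_07, rooted on mstr_ribbon.
# The mstr_ribbon body adds a larger offset box alongside its capsule.
# Identical constraints link each neighboring pair, allowing moderate swing with very little twist.
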PhysicsAsset /BRCosmetics/Weapons/FORT_Melee/Pickaxe_Cuddle_Team_Dark/Meshes/Cuddle_Team_Dark_Physics
bodies {
 dyn_ribbon_01 [{shapeType=Capsule, length=0, radius=1, center=<1 0 0>}]
 dyn_ribbon_02 [{shapeType=Capsule, length=0, radius=1, center=<1 0 0>}]
 dyn_ribbon_03 [{shapeType=Capsule, length=0, radius=1, center=<1 0 0>}]
 dyn_ribbon_04 [{shapeType=Capsule, length=0, radius=1, center=<1 0 0>}]
 dyn_ribbon_05 [{shapeType=Capsule, length=0, radius=1, center=<1 0 0>}]
 dyn_ribbon_06 [{shapeType=Capsule, length=0, radius=1, center=<1 0 0>}]
 dyn_ribbon_07 [{shapeType=Capsule, length=0, radius=1, center=<1 0 0>}]
 mstr_ribbon [{shapeType=Capsule, length=0, radius=1, center=<1 0 0>}, {shapeType=Box, extentX=14, extentY=3, extentZ=16, center=<2.4436 -2.6761 5.7078>}]
}
constraints [
 {bone1Name=dyn_ribbon_03, bone2Name=dyn_ribbon_02, swing1LimitDegrees=45, swing2LimitDegrees=45, twistLimitDegrees=2.5}
 {bone1Name=dyn_ribbon_04, bone2Name=dyn_ribbon_03, swing1LimitDegrees=45, swing2LimitDegrees=45, twistLimitDegrees=2.5}
 {bone1Name=dyn_ribbon_05, bone2Name=dyn_ribbon_04, swing1LimitDegrees=45, swing2LimitDegrees=45, twistLimitDegrees=2.5}
 {bone1Name=dyn_ribbon_06, bone2Name=dyn_ribbon_05, swing1LimitDegrees=45, swing2LimitDegrees=45, twistLimitDegrees=2.5}
 {bone1Name=dyn_ribbon_07, bone2Name=dyn_ribbon_06, swing1LimitDegrees=45, swing2LimitDegrees=45, twistLimitDegrees=2.5}
 {bone1Name=dyn_ribbon_02, bone2Name=dyn_ribbon_01, swing1LimitDegrees=45, swing2LimitDegrees=45, twistLimitDegrees=2.5}
 {bone1Name=dyn_ribbon_01, bone2Name=mstr_ribbon, swing1LimitDegrees=45, swing2LimitDegrees=45, twistLimitDegrees=2.5}
]
